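# Animation Blueprint for the FermiumStraw pickaxe cosmetic, adding secondary physics motion to its tube bones.
# The input pose passes through four chained world-space AnimDynamics nodes (dyn_tube_01 to 04), each angularly sprung, limit-constrained and using overridden gravity.
# Each update sets a linear velocity scale from zipline state and a simulation space from front-end state.
# The AnimDynamics nodes show fixed values rather than reading those two variables.
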AnimBlueprint /BRCosmetics/Weapons/FORT_Melee/Pickaxe_FermiumStraw/Meshes/FermiumStraw_Axe_AnimBP
target_skeleton=&/FermiumStraw_Axe_Skeleton
graphs [
 AnimGraph {nodes [
   AnimGraphNode_SubInput_0 = Input Pose()
   n1 = Anim Dynamics - Bone: dyn_tube_01!(
    SimulationSpace:AnimPhysSimSpaceType Enum=World
    ComponentPose=Local To Component(LocalPose=@AnimGraphNode_SubInput_0)
    Alpha=1
    ComponentLinearAccScale:Vector="1,1,1"
    ComponentLinearVelScale:Vector="1,1,1"
    ExternalForce="0,0,0"
   ) -> (Pose) {
    angularDampingOverride=1
    angularSpringConstant=11000
    bAngularSpring=true
    bOverrideAngularDamping=true
    bUseGravityOverride=true
    gravityScale=0.35
    simulationSpace=World
    alphaBoolBlend {blendOption=Linear}
    boundBone {boneName=dyn_tube_01}
    componentAppliedLinearAccClamp=<9500 9500 9500>
    componentLinearAccScale=<1 1 1>
    componentLinearVelScale=<1 1 1>
    gravityOverride {z=-900}
    physicsBodyDefinitions [{
      boundBone {boneName=dyn_tube_01}
      boxExtents=<18 18 18>
      constraintSetup {
       angularLimitsMax=<18 20 19>
       angularLimitsMin=<-18 -20 -19>
       angularTarget=<1 0 0>
       linearAxesMax=<0.4 0.8 0.5>
       linearAxesMin=<-0.8 -0.4 -0.5>
      }
      localJointOffset=<1 0 0>
     }]
    simSpaceSettings {maxAngularAcceleration=35, maxAngularVelocity=7, simSpaceAngularAlpha=1}
   }
   n2 = Anim Dynamics - Bone: dyn_tube_02!(
    SimulationSpace:AnimPhysSimSpaceType Enum=World
    ComponentPose=@n1
    Alpha=1
    ComponentLinearAccScale:Vector="1,1,1"
    ComponentLinearVelScale:Vector="1,1,1"
    ExternalForce="0,0,0"
   ) -> (Pose) {
    angularDampingOverride=1
    angularSpringConstant=11000
    bAngularSpring=true
    bOverrideAngularDamping=true
    bUseGravityOverride=true
    gravityScale=0.35
    simulationSpace=World
    alphaBoolBlend {blendOption=Linear}
    boundBone {boneName=dyn_tube_02}
    componentAppliedLinearAccClamp=<9500 9500 9500>
    componentLinearAccScale=<1 1 1>
    componentLinearVelScale=<1 1 1>
    gravityOverride {z=-980}
    physicsBodyDefinitions [{
      boundBone {boneName=dyn_tube_02}
      boxExtents=<20 20 20>
      constraintSetup {
       angularLimitsMax=<20 20 20>
       angularLimitsMin=<-20 -20 -20>
       angularTarget=<1 0 0>
       linearAxesMax=<0.5 0.5 0.5>
       linearAxesMin=<-0.5 -0.5 -0.5>
      }
      localJointOffset=<1 0 0>
     }]
    simSpaceSettings {maxAngularAcceleration=28, maxAngularVelocity=7, simSpaceAngularAlpha=1}
   }
   n3 = Anim Dynamics - Bone: dyn_tube_03!(
    SimulationSpace:AnimPhysSimSpaceType Enum=World
    ComponentPose=@n2
    Alpha=1
    ComponentLinearAccScale:Vector="1,1,1"
    ComponentLinearVelScale:Vector="1,1,1"
    ExternalForce="0,0,0"
   ) -> (Pose) {
    angularDampingOverride=1
    angularSpringConstant=11000
    bAngularSpring=true
    bOverrideAngularDamping=true
    bUseGravityOverride=true
    gravityScale=0.35
    simulationSpace=World
    alphaBoolBlend {blendOption=Linear}
    boundBone {boneName=dyn_tube_03}
    componentAppliedLinearAccClamp=<9500 9500 9500>
    componentLinearAccScale=<1 1 1>
    componentLinearVelScale=<1 1 1>
    gravityOverride {z=-1010}
    physicsBodyDefinitions [{
      boundBone {boneName=dyn_tube_03}
      boxExtents=<15 18 22>
      constraintSetup {
       angularLimitsMax=<15 22 18>
       angularLimitsMin=<-15 -22 -18>
       angularTarget=<1 0 0>
       linearAxesMax=<0.3 0.4 0.5>
       linearAxesMin=<-0.3 -0.4 -0.5>
      }
      localJointOffset=<1 0 0>
     }]
    simSpaceSettings {maxAngularAcceleration=30, maxAngularVelocity=7, simSpaceAngularAlpha=1}
   }
   n4 = Anim Dynamics - Bone: dyn_tube_04!(
    SimulationSpace:AnimPhysSimSpaceType Enum=World
    ComponentPose=@n3
    Alpha=1
    ComponentLinearAccScale:Vector="1,1,1"
    ComponentLinearVelScale:Vector="1,1,1"
    ExternalForce="0,0,0"
   ) -> (Pose) {
    angularDampingOverride=1
    angularSpringConstant=11000
    bAngularSpring=true
    bOverrideAngularDamping=true
    bUseGravityOverride=true
    gravityScale=0.35
    simulationSpace=World
    alphaBoolBlend {blendOption=Linear}
    boundBone {boneName=dyn_tube_04}
    componentAppliedLinearAccClamp=<9500 9500 9500>
    componentLinearAccScale=<1 1 1>
    componentLinearVelScale=<1 1 1>
    gravityOverride {z=-980}
    physicsBodyDefinitions [{
      boundBone {boneName=dyn_tube_04}
      boxExtents=<22 22 22>
      constraintSetup {
       angularLimitsMax=<25 25 25>
       angularLimitsMin=<-25 -25 -25>
       angularTarget=<1 0 0>
       linearAxesMax=<0.7 0.6 0.8>
       linearAxesMin=<-0.7 -0.6 -0.8>
      }
      localJointOffset=<1 0 0>
     }]
    simSpaceSettings {maxAngularAcceleration=35, maxAngularVelocity=7, simSpaceAngularAlpha=1}
   }
   Output Pose(Result=Component To Local(ComponentPose=@n4)) {layerGroup=None}
  ]}
 EventGraph {nodes [
   Event BlueprintUpdateAnimation()
   > n1 = Sequence()
   n2 = GetOwnerAnimationInterface!(self:Custom Character Part Anim Instance Object Reference) -> (ReturnValue:Fort Player Animation Interface Object Reference)
   > Is Valid(InputObject=@n2)
   > n3 = Branch(Condition=Break Fort Anim Input Zipline!(FortAnimInput_Zipline:"Fort Anim Input Zipline Structure (by ref)"=GetZiplineInput!(self:Fort Player Animation Interface Object Reference=@n2) -> (ReturnValue:"Fort Anim Input Zipline Structure (by ref)")) -> (
     bIsZiplining:bool=~
     bIsOnAscender:bool=~
     bShouldPlayPivotTransition:bool=~
     LeanYaw:float=~
     PivotBlendDelayRemaining:float=~
     PivotBlendDelay:float=~
     PivotCardinalDirection:EFortCardinalDirection Enum
     WorldVelocityLastTick:Vector=~
    ) {is=title K2Node_BreakStruct})
   > Set Componen Linear Vel Scale(Componen Linear Vel Scale:Vector="0, 0, 0") -> (Output_Get:Vector)
   Set Componen Linear Vel Scale(execute=@n3.else, Componen Linear Vel Scale="0.5,0.5,0.5")
   n4 = Branch(execute=@n1.then_1, Condition=Get bIsInFrontEnd())
   > Set AnimDynamic_SimulationSpace(AnimDynamic_SimulationSpace:Byte=1) -> (Output_Get:Byte=0)
   Set AnimDynamic_SimulationSpace(execute=@n4.else, AnimDynamic_SimulationSpace=0) -> (Output_Get=0)
   TryGetPawnOwner()
  ]}
]
variables [
 AnimDynamic_SimulationSpace {type=integer}
 Componen Linear Vel Scale {type=Vector, default=<0.5 0.5 0.5>}
]
